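# Material function that sets the emissive color for the OctanePod skin on the Octane vehicle body.
# A selectable SkinMask channel, alpha by default, multiplies SkinEmissiveColor, and the result is raised to EmissivePower, saturated, then scaled by EmissiveIntensity.
# The result replaces EmissiveColor in the incoming material attributes and passes the others through.
# The default black color produces no glow unless overridden.
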MaterialFunction /VehicleCosmetics/Bodies/Octane/Skins/OctanePod/MF_Skin_OctanePod_Emissive
graph {nodes [
  n1 = FunctionInput() {inputType=FunctionInput_MaterialAttributes}
  n2 = ScalarParameter() {defaultValue=3, group=Emissive, parameterName=EmissiveIntensity}
  n3 = VectorParameter() {group=Emissive, parameterName=SkinEmissiveColor, defaultValue=<rgba 0 0 0 1>}
  n4 = TextureSampleParameter2D() {group=Texture, parameterName=SkinMask, samplerSource=SSM_Wrap_WorldGroupSettings, texture=/VehicleCosmetics/Textures/T_Black}
  n5 = ChannelMaskParameter(None=@n4.RGBA) {maskChannel=Alpha, parameterName=EmissiveMask, defaultValue=<rgba 0 0 0 1>}
  n6 = Saturate(None=Power(Base=Multiply(A=@n3.RGB, B=@n5.R), Exp=ScalarParameter() {defaultValue=1.3, group=Emissive, parameterName=EmissivePower}))
  FunctionOutput(None=SetMaterialAttributes!(MaterialAttributes=GetMaterialAttributes(None=@n1).MaterialAttributes, EmissiveColor=Multiply(A=@n2, B=@n6)) {attributeSetTypes [#1]})
 ]}
